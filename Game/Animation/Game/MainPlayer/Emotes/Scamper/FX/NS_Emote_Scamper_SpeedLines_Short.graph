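# Niagara sprite system that draws short speed lines for the player's Scamper emote.
# One emitter spawns two groups in boxes either side of the origin, with rates tapering late in the loop and zeroed by User.bIsInWater.
# Short-lived sprites move along -Y, grow and stretch with speed, and fade by age and by the first box's normalized spawn Z.
# User.ScaleLines is declared, but no shown module reads it.
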NiagaraSystem /Game/Animation/Game/MainPlayer/Emotes/Scamper/FX/NS_Emote_Scamper_SpeedLines_Short
dependencies {
 usedDataInterfaces [&NiagaraDataInterfaceCurve, &NiagaraDataInterfaceVector2DCurve, &NiagaraDataInterfaceColorCurve]
 usedDynamicInputs []
 usedModules [&SystemState]
}
emitters {SpeedLines002 {
  emitterUpdateScript [
   &EmitterState()
   &SpawnRate(
    SpawnRate=&FloatFromCurve(
     FloatCurve=DataInterface() {propertyValues {Curve {keys [{interpMode=RCIM_Cubic, tangentMode=RCTM_User, time=0.5, value=1}, {interpMode=RCIM_Cubic, tangentMode=RCTM_User, time=1, value=0}]}}}
     CurveIndex=$Emitter.NormalizedLoopAge
     Scale Curve=&Multiply_Float(
      A=&MakeCustomFloatFromBool(Bool=$User.bIsInWater, True Float=0, False Float=1)
      B=&RandomRangeFloat(Minimum=15, Maximum=25)
     )
    )
    Use Spawn Probability=false
   ) {module=SpawnRate001}
   &SpawnRate(
    SpawnGroup=1
    SpawnRate=&FloatFromCurve(
     FloatCurve=DataInterface() {propertyValues {Curve {keys [{interpMode=RCIM_Cubic, tangentMode=RCTM_User, time=0.5, value=1}, {interpMode=RCIM_Cubic, tangentMode=RCTM_User, time=1, value=0}]}}}
     CurveIndex=$Emitter.NormalizedLoopAge
     Scale Curve=&Multiply_Float(
      A=&MakeCustomFloatFromBool(Bool=$User.bIsInWater, True Float=0, False Float=1)
      B=&RandomRangeFloat(Minimum=15, Maximum=25)
     )
    )
    Use Spawn Probability=false
   )
  ]
  particleSpawnScript [
   &FN_SetSpawnGroup(Number of Spawn Groups=2)
   &InitializeParticle(
    Color=&MakeLinearColorFromVectorAndFloat("Vector (RGB)"=<1 1 1>, "Float (Alpha)"=&RandomRangeFloat(Minimum=0.3, Maximum=0.5))
    Lifetime Max=0.16
    Lifetime Min=0.1
    Lifetime Mode=::NewEnumerator1
    Sprite Size Max=<3 10>
    Sprite Size Min=<2 2>
    Sprite Size Mode=::NewEnumerator2
    UsePositionOffset=false
    Write Material Random=true
   )
   &ShapeLocation(
    Box Midpoint=<0.5 -0.5 0.5>
    Box Size=<20 20 80>
    MaskSpawnGroup=true
    Offset=<30 0 70>
    Offset Mode=::NewEnumerator0
    Override Seed=false
    Shape Primitive=::NewEnumerator2
    Surface Only=false
    UseOwnerScale=true
   )
   &ShapeLocation(
    Box Midpoint=<0.5 -0.5 0.5>
    Box Size=<20 20 80>
    MaskSpawnGroup=true
    Offset=<-30 0 70>
    Offset Mode=::NewEnumerator0
    Override Seed=false
    Shape Primitive=::NewEnumerator2
    Spawn Group Mask=1
    Surface Only=false
    UseOwnerScale=true
   ) {module=ShapeLocation001}
   /SetVariables(Particles.Position=$Particles.ShapeLocation001.ShapePosition) {enabled=false, isSetParameters=true}
   &AddVelocity(Velocity=&RandomRangeVector(Minimum=<0 -100 0>, Maximum=<0 -250 0>))
   /SetVariables(Particles.ShapePos:Vector3f=&NormalizeVector(Vector To Normalize=$Particles.ShapeLocation.ShapePosition, Safe Normalize=true, Safe Fallback Threshold=0, Safe Fallback Vector=<0 0 0.001>)) {isSetParameters=true}
  ]
  particleUpdateScript [
   &SolveForcesAndVelocity(Clamp Velocity=false, Limit Acceleration=false)
   &ParticleState()
   &ScaleSpriteSize(Scale Factor=&Vector2DFromCurve(
     Vector2Curve=DataInterface() {propertyValues {XCurve {keys [<at 0 0.5 RCIM_Cubic>,<at 0.5 1 RCIM_Cubic>]}, YCurve {keys [<at 0 0 RCIM_Cubic>,<at 0.5 1 RCIM_Cubic>]}}}
     ScaleCurve=&Vector2DFromFloat(Value=1)
    ))
   &ScaleSpriteSizeBySpeed(Max Scale Factor=<1 10>, Min Scale Factor=<0.01 0.05>, Velocity Threshold=1)
   &ScaleColor(
    Scale Alpha=&Multiply_Float(A=&MakeFloatFromVector(Vector=$Particles.ShapePos, Channel=::NewEnumerator1), B=5)
    ScaleA=true
    ScaleRGB=true
   ) {enabled=false, module=ScaleColor001}
   &ScaleColor(
    Scale Alpha=&FloatFromCurve(
     FloatCurve=DataInterface() {propertyValues {Curve {keys [
         {arriveTangent=-3.5623, interpMode=RCIM_Cubic, leaveTangent=-3.5623, tangentMode=RCTM_User, time=0, value=1}
         {arriveTangent=-0.3548, tangentMode=RCTM_User, time=1, value=0}
        ]}}}
     CurveIndex=&MakeFloatFromVector(Vector=$Particles.ShapePos, Channel=::NewEnumerator2)
     Scale Curve=15
    )
    ScaleA=true
    ScaleRGB=true
   ) {module=ScaleColor002}
   &ScaleColor(
    Scale Alpha=&FloatFromCurve(FloatCurve=DataInterface() {propertyValues {Curve {keys [<at 0 0 RCIM_Cubic>, <at 0.4 1>, {arriveTangent=-2.221, interpMode=RCIM_Cubic, leaveTangent=-2.221, tangentMode=RCTM_User, time=1, value=0}]}}})
    Scale RGB=&VectorFromFloat(Value=1)
    ScaleA=true
    ScaleRGB=true
   )
  ]
  rendererClasses [&NiagaraSpriteRendererProperties]
  renderers [{rendererClass=&NiagaraSpriteRendererProperties}]
 }}
summary {
 systemName=NS_Emote_Scamper_SpeedLines_Short
 emitters [{emitterName=SpeedLines002, rendererClasses [&NiagaraSpriteRendererProperties]}]
 userVariables [
  User.bIsInWater {type=NiagaraBool}
  User.ScaleLines {type=NiagaraFloat, defaultValue=1, description=""}
 ]
}
